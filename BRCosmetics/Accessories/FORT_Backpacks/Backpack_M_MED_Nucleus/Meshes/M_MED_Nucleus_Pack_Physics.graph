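# Physics asset for the M_MED_Nucleus backpack cosmetic, defining bodies and joints for its dangling pistol, lace and zipper bones.
# A small capsule on the pack bone is the hub, with box bodies for left and right pistols, both laces and the zipper each constrained directly to it.
# Both lace joints lock all swing and twist, while pistol and zipper joints do not, and the two pistols have asymmetric swing angles.
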PhysicsAsset /BRCosmetics/Accessories/FORT_Backpacks/Backpack_M_MED_Nucleus/Meshes/M_MED_Nucleus_Pack_Physics
bodies {
 dyn_lace_l [{shapeType=Box, extentX=3.5906, extentY=1.8388, extentZ=7.7954, shapeName=dyn_lace_l_box, rotation=<rot 0 -180 -90>}]
 dyn_lace_r [{shapeType=Box, extentX=3.5906, extentY=1.8388, extentZ=7.7954, shapeName=dyn_lace_l_box, rotation=<rot 0 180 90>}]
 dyn_pistol_l [{shapeType=Box, extentX=10.7287, extentY=3.6179, extentZ=20.0697, shapeName=dyn_pistol_l_box, center=<1 0 0>, rotation=<rot 0 90 -90>}]
 dyn_pistol_r [{shapeType=Box, extentX=10.7287, extentY=3.6179, extentZ=20.0697, shapeName=dyn_pistol_l_box, center=<1 0 0>, rotation=<rot 0 90 90>}]
 dyn_zipper [{
   shapeType=Box
   extentX=1.5237
   extentY=0.4296
   extentZ=3.0949
   shapeName=dyn_zipper_box
   center=<2.049 0.026 -0.1155>
   rotation=<rot 0 90.8049 -89.9391>
  }]
 pack [{shapeType=Capsule, length=5, radius=2, shapeName=pack_capsule}]
}
constraints [
 {bone1Name=dyn_pistol_l, bone2Name=pack, swing1LimitDegrees=45, swing2LimitDegrees=10, twistLimitDegrees=20}
 {bone1Name=dyn_pistol_r, bone2Name=pack, swing1LimitDegrees=25, swing2LimitDegrees=10, twistLimitDegrees=20}
 {
  bone1Name=dyn_lace_l
  bone2Name=pack
  swing1LimitDegrees=5
  swing1Motion=Locked
  swing2LimitDegrees=45
  swing2Motion=Locked
  twistLimitDegrees=45
  twistMotion=Locked
 }
 {
  bone1Name=dyn_lace_r
  bone2Name=pack
  swing1LimitDegrees=5
  swing1Motion=Locked
  swing2LimitDegrees=45
  swing2Motion=Locked
  twistLimitDegrees=45
  twistMotion=Locked
 }
 {bone1Name=dyn_zipper, bone2Name=pack, swing1LimitDegrees=45, swing2LimitDegrees=50, twistLimitDegrees=5}
]
