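# Animation Blueprint for the Nucleus backpack cosmetic, adding procedural secondary motion to the pack.
# From the local ref pose, a RigidBody node simulates the pack's physics asset in root-bone space at partial alpha, with custom gravity and scaled component motion.
# AnimDynamics then swings the zipper bone within angular limits under heavy damping, and a Modify Bone node additively scales the pack bone.
# The event graph holds only unconnected stub nodes.
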AnimBlueprint /BRCosmetics/Accessories/FORT_Backpacks/Backpack_M_MED_Nucleus/Meshes/M_MED_Nucleus_Pack_AnimBP
target_skeleton=&/M_MED_Nucleus_Pack_Skeleton
graphs [
 AnimGraph {nodes [
   n1 = RigidBody(ComponentPose=Local To Component(LocalPose=Local Space Ref Pose()), Alpha=0.8) {
    bOverrideWorldGravity=true
    overridePhysicsAsset=&/M_MED_Nucleus_Pack_Physics
    simulationSpace=BaseBoneSpace
    alphaBoolBlend {blendOption=Linear}
    baseBoneRef {boneName=root}
    componentLinearAccScale=<0.5 0.5 0.75>
    componentLinearVelScale=<0.5 0.5 0.75>
    overrideWorldGravity {z=-300}
   }
   n2 = Anim Dynamics - Bone: dyn_zipper(ComponentPose=@n1, Alpha=1) {
    angularDampingOverride=0.99
    bGravityOverrideInSimSpace=true
    bOverrideAngularDamping=true
    simulationSpace=RootRelative
    alphaBoolBlend {blendOption=Linear}
    boundBone {boneName=dyn_zipper}
    componentAppliedLinearAccClamp=<5000 5000 5000>
    componentLinearAccScale=<1 1 1>
    componentLinearVelScale=<1 1 1>
    physicsBodyDefinitions [{
      sphereCollisionRadius=0
      boundBone {boneName=dyn_zipper}
      boxExtents=<3 10 10>
      constraintSetup {angularLimitsMax=<2 75 45>, angularLimitsMin=<-2 -10 -45>, angularTarget=<0 0 0>}
      localJointOffset=<1.5 -0.05 0.1>
     }]
   }
   n3 = "Transform (Modify) Bone - Bone: pack"(ComponentPose=@n2, Translation="0,0,0", Rotation="0,0,0", Scale="0.9,0.9,0.9") {scaleMode=BMM_Additive, alphaBoolBlend {blendOption=Linear}, boneToModify {boneName=pack}, scale=<0.9 0.9 0.9>}
   Output Pose(Result=Component To Local(ComponentPose=@n3)) {layerGroup=None}
  ]}
 EventGraph {nodes [
   TryGetPawnOwner()
   Event BlueprintUpdateAnimation()
  ]}
]
variables []
